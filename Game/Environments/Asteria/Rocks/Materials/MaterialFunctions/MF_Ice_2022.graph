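# Material function for Asteria rock ice, outputting subsurface-lit material attributes plus a noise mask.
# World-aligned rock detail and ice noise supply layered normals and depth, flattened on upward faces.
# Ice tones blend through masks built from vertex colors (Nanite) or a mask texture, Fresnel and an RVT input.
# Low, Medium and High quality fold subsurface color into base color; mobile uses time-of-day-scaled emissive instead.
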MaterialFunction /Game/Environments/Asteria/Rocks/Materials/MaterialFunctions/MF_Ice_2022
graph {nodes [
  n1 = VectorParameter() {group=Ice, parameterName=IceTone Emissive Inside, sortPriority=3, defaultValue=<rgba 0 0.0987 0.224 0.1>}
  n2 = Multiply(A=@n1.RGB, B=@n1.RGB)
  n3 = NamedRerouteUsage()
  n4 = ScalarParameter() {defaultValue=0.25, group=Ice, parameterName=Ice Spec}
  n5 = Multiply(A=@n4) {constB=0.5}
  n6 = NamedRerouteUsage()
  n7 = Constant() {r=0.1}
  n8 = Constant() {r=1.5}
  n9 = VectorParameter() {parameterName=IceTone Bottom SSS, defaultValue=<rgba 0.0699 0.4282 0.6198 0>}
  n10 = VectorParameter() {group=Ice, parameterName=IceTone Emissive Outside, sortPriority=2, defaultValue=<rgba 0.2558 0.6704 0.908 0.5>}
  n11 = LinearInterpolate(
   A=LinearInterpolate(A=Multiply(A=@n8, B=@n9.RGB), B=@n9.RGB, Alpha=NamedRerouteUsage())
   B=LinearInterpolate(A=@n1.RGB, B=@n10.RGB, Alpha=NamedRerouteUsage())
   Alpha=@n6
  )
  n12 = ScalarParameter() {defaultValue=1, parameterName=SSS Value}
  n13 = Constant() {r=0.5}
  n14 = Constant2Vector() {g=-1, r=-1}
  n15 = ScalarParameter() {defaultValue=0.7, group=Ice, parameterName=ScaleN, sortPriority=1}
  n16 = Multiply(A=@n15, B=Constant() {r=2000})
  n17 = MaterialFunctionCall!(
   Diffuse Texture Object
   Spec Texture Object=TextureObject() {samplerType=SAMPLERTYPE_Masks, texture=/Game/Environments/Asteria/Rocks/Textures/T_Rock_BigDetails_M}
   TextureSize=@n16
   --------------------------
   Use seperate settings for Z projection
   Use seperate texture for Z projection
   Texture Size Z projection=@n16
   Z Plane Diffuse Texture Object
   Z Plane Spec Texture Object
   "--------------------------      "
   ProjectionTransitionContrast
   Use Texture Alpha Heightmap to Refine the Transistion
   HeightLerpTransitionContrast
   PickTheDiffuseHeightChannel
   TheDiffuseHeightChannel
   "--------------------------       "
   NormalTextureObject=TextureObject() {samplerType=SAMPLERTYPE_Normal, texture=/Game/Environments/Asteria/Rocks/Textures/T_Rock_BigDetails_N}
   World Space Normals=StaticBool()
   "--------------------------  "
   Export Float 4 Diffuse
   Export Float 4 Spec
   "--------------------------     "
   WorldPosition
   World Space Normal
  ) {materialFunction=/Engine/Functions/Engine_MaterialFunctions02/Texturing/WorldAlignedTextures_Complex}
  n18 = AppendVector(
   A=Multiply(A=Multiply(A=@n13, B=@n14), B=ComponentMask(None=@n17.Normal XYZ Flat Top) {g=true, r=true})
   B=ComponentMask(None=@n17.Normal XYZ Flat Top) {b=true}
  )
  n19 = TextureObject() {samplerType=SAMPLERTYPE_Masks, texture=/Game/Environments/Asteria/Terrain/Textures/T_Asteria_Ice_Noise_M}
  n20 = MaterialFunctionCall(
   Spec Texture Object=@n19
   TextureSize=Multiply(A=Constant() {r=800}, B=@n15)
   Use seperate settings for Z projection=StaticBool() {value=true}
   Texture Size Z projection=ScalarParameter() {defaultValue=400, parameterName=Up Noise out}
   Z Plane Spec Texture Object=@n19
   NormalTextureObject=TextureObject() {samplerType=SAMPLERTYPE_Normal, texture=/Game/Environments/Asteria/Terrain/Textures/T_Asteria_Ice_SubtleNoise_N}
   World Space Normals=StaticBool()
  ) {materialFunction=&WorldAlignedTextures_Complex}
  n21 = Multiply(
   A=ScalarParameter() {defaultValue=3, parameterName=Noise2}
   B=Multiply(A=@n20.Normal XYZ Flat Top, B=Constant3Vector() {constant=<rgba 1 1 0 0>})
  )
  n22 = LinearInterpolate(
   A=Constant() {r=-0.408}
   B=Constant() {r=1}
   Alpha=LinearInterpolate(B=NamedRerouteUsage(), Alpha=ScalarParameter() {defaultValue=1, group=Ice, parameterName=Ice Normals Mask}) {constA=1}
  )
  n23 = SetMaterialAttributes!(
   MaterialAttributes
   BaseColor=@n2
   Specular=Multiply(A=OneMinus(None=@n3), B=LinearInterpolate(A=@n5, B=@n4, Alpha=@n6))
   Roughness=Multiply(A=LinearInterpolate(B=@n7, Alpha=@n6) {constA=0.25}, B=VertexColor().G)
   SubsurfaceColor=NamedRerouteDeclaration(None=Multiply(A=@n11, B=@n12)) {name=E, nodeColor=<rgba 0 0.2266 1 1>}
   Normal=LinearInterpolate(A=@n18, B=Add(A=@n21, B=@n18), Alpha=Saturate(None=@n22))
   Opacity=Constant() {r=0.5}
  ) {attributeSetTypes [
    #1
    #2
    #3
    #4
    #5
    #6
   ]}
  n24 = Saturate(None=ComponentMask(None=VertexNormalWS()) {b=true})
  n25 = BlendMaterialAttributes(
   A=@n23
   B=SetMaterialAttributes!(MaterialAttributes=@n23, Normal=Constant3Vector() {constant=<rgba 0 0 1 0>}, BaseColor=@n2) {attributeSetTypes [#5, #1]}
   Alpha=NamedRerouteDeclaration(None=@n24) {name=Top Mask, nodeColor=<rgba 1 0 0.0234 1>}
  )
  n26 = GetMaterialAttributes(None=@n25) {attributeGetTypes [#4, #1, #3]}
  n27 = VectorParameter() {
   desc="Mult Emissive by This color per TOD "
   group=A_Emissive_TOD
   parameterName=EmissiveMorning
   sortPriority=1
   defaultValue=<rgba 0.0404 0.05 0.0133 1>
  }
  n28 = VectorParameter() {
   desc="Mult Emissive by This color per TOD "
   group=A_Emissive_TOD
   parameterName=EmissiveDay
   sortPriority=2
   defaultValue=<rgba 0 0.2 0.1475 0>
  }
  n29 = VectorParameter() {
   desc="Mult Emissive by This color per TOD "
   group=A_Emissive_TOD
   parameterName=EmissiveEvening
   sortPriority=3
   defaultValue=<rgba 0.1989 0.2 0.1219 1>
  }
  n30 = VectorParameter() {
   desc="Mult Emissive by This color per TOD "
   group=A_Emissive_TOD
   parameterName=EmissiveNight
   sortPriority=4
   defaultValue=<rgba 0.0704 0.15 0.1453 0>
  }
  n31 = Multiply(
   A=MaterialFunctionCall(Color_Morning=@n27.RGB, Color_Day=@n28.RGB, Color_Evening=@n29.RGB, Color_Night=@n30.RGB) {materialFunction=&Color_Time_of_Day}.Result
   B=NamedRerouteUsage()
  )
  n32 = Multiply(A=@n31, B=@n26.SubsurfaceColor)
  n33 = ScalarParameter() {
   defaultValue=0.6
   desc="Lerp Control, 0 = Pure SSC grab, 1 = Pure ICE_SSC_Fallback_Trendcolor"
   group=Ice
   parameterName=Ice_SSC_Fallback_Blend
  }
  n34 = SetMaterialAttributes!(
   MaterialAttributes=@n26.MaterialAttributes
   BaseColor=LinearInterpolate(A=LinearInterpolate(A=@n26.SubsurfaceColor, B=@n32, Alpha=@n33) {constAlpha=0.1}, B=@n26.BaseColor, Alpha=@n26.BaseColor)
  ) {attributeSetTypes [#1]}
  n35 = GetMaterialAttributes(None=@n25) {attributeGetTypes [#4]}
  n36 = Multiply(A=@n35.SubsurfaceColor, B=@n31)
  n37 = Saturate(None=DotProduct(A=CameraVectorWS(), B=VertexNormalWS()))
  n38 = ShadingPathSwitch(
   Default=QualitySwitch(Default=@n25, Low=@n34, High=@n34, Medium=@n34)
   Mobile=SetMaterialAttributes!(MaterialAttributes=@n35.MaterialAttributes, EmissiveColor=Multiply(A=@n36, B=@n37)) {attributeSetTypes [#7]}
  )
  n39 = GetMaterialAttributes(None=@n38) {attributeGetTypes [#4]}
  n40 = Constant() {r=0.2}
  n41 = Constant() {r=0.7}
  n42 = LinearInterpolate(
   A=Multiply(
    A=@n39.SubsurfaceColor
    B=MaterialFunctionCall(Color_Morning=@n40, Color_Day=@n41, Color_Evening=@n40, Color_Night=@n40) {materialFunction=&Color_Time_of_Day}.Result
   )
   Alpha=NamedRerouteUsage()
  ) {constB=0}
  FunctionOutput(None=ShadingPathSwitch(
    Default=@n38
    Mobile=SetMaterialAttributes(MaterialAttributes=@n39.MaterialAttributes, EmissiveColor=@n42) {attributeSetTypes [#7]}
   ) {desc=Migrated from FeatureLevelSwitch})
  FunctionOutput(None=ComponentMask(None=@n20.XYZ Spec) {r=true}) {outputName=Noise}
  n43 = NamedRerouteUsage()
  n44 = NamedRerouteUsage()
  n45 = ScalarParameter() {group=Ice, parameterName="Add Soft Curvature "}
  n46 = StaticSwitchParameter(True=TextureCoordinate() {coordinateIndex=2}, False=TextureCoordinate()) {group=Ice Fallback, parameterName=Use UV2 for Fallback}
  n47 = TextureSampleParameter2D(UVs=@n46) {
   group=Ice Fallback
   parameterName=Ice Mask RGB
   samplerType=SAMPLERTYPE_Masks
   sortPriority=2
   texture=/Game/Global/Textures/Default/Blanks/Temp_specular_Mask
   channelNames {a="", b="", g=AO, r=Curvature}
  }
  n48 = Saturate(None=Max(A=Multiply(A=@n45, B=@n47.Curvature), B=@n47.B))
  n49 = VertexColor()
  n50 = Multiply(A=ScalarParameter() {group=Ice, parameterName="Add Soft Curvature "}, B=@n49.R)
  n51 = Max(A=@n50, B=@n49.B)
  n52 = Add(A=@n50, B=@n49.B)
  n53 = StaticSwitchParameter(
   True=Saturate(None=StaticSwitchParameter(True=@n51, False=@n52) {group=Ice, parameterName=Use Max Blending R to B channel, sortPriority=-3})
   False=@n48
  ) {
   desc="Disable to set up fallback material if using vertex colors, and enable back after that. "
   group=Ice
   parameterName=Use VertexColor Mask
   sortPriority=-1
  }
  n54 = NaniteReplace!(Default=@n48, Nanite=@n53)
  n55 = TextureSampleParameter2D(UVs=@n46) {
   group=Ice Fallback
   parameterName=Ice Mask RGB
   samplerType=SAMPLERTYPE_Masks
   sortPriority=2
   texture=&Temp_specular_Mask
   channelNames {a="", b="", g=AO, r=Curvature}
  }
  n56 = VertexColor()
  n57 = Multiply(
   A=Multiply(A=LinearInterpolate(A=@n44, Alpha=@n54), B=NamedRerouteDeclaration(None=@n54) {name=Thick, nodeColor=<rgba 1 0 0.2109 1>})
   B=NaniteReplace(Default=@n55.AO, Nanite=StaticSwitchParameter(True=@n56.G, False=@n55.AO) {group=Ice, parameterName=Use VertexColor Mask, sortPriority=-1})
  )
  n58 = NamedRerouteUsage()
  n59 = NamedRerouteUsage()
  NamedRerouteDeclaration(None=Multiply(A=Multiply(A=@n43, B=@n57), B=LinearInterpolate(B=@n58, Alpha=@n59) {constA=1})) {name=DepthMask, nodeColor=<rgba 1 0 0.8672 1>}
  n60 = Constant() {r=0.65}
  n61 = NamedRerouteDeclaration(None=ComponentMask(None=@n20.XY Spec) {g=true}) {name=Noise, nodeColor=<rgba 0.8516 0 1 1>}
  n62 = Constant() {r=8}
  n63 = Constant() {r=3}
  n64 = VertexNormalWS()
  NamedRerouteDeclaration(None=LinearInterpolate(A=Max(A=@n60, B=@n61), B=Multiply(A=@n61, B=@n62), Alpha=Fresnel(ExponentIn=@n63, Normal=@n64) {baseReflectFraction=0})) {name=Tiled Noise, nodeColor=<rgba 0.2422 1 0 1>}
  n65 = Constant() {r=1}
  n66 = Divide(A=FunctionInput() {inputName=RVT, inputType=FunctionInput_Scalar}, B=Constant() {r=800})
  n67 = VirtualTextureFeatureSwitch!(No=@n65, Yes=MaterialProxyReplace(Realtime=Saturate(None=@n66), MaterialProxy=@n65))
  NamedRerouteDeclaration(None=ShadingPathSwitch(Default=@n67, Deferred=@n67, Forward=@n65, Mobile=@n65)) {name=RVT, nodeColor=<rgba 0 0.1094 1 1>}
  n68 = Constant() {r=1.35}
  n69 = OneMinus(None=ComponentMask(None=@n17.XY Spec) {g=true})
  n70 = Saturate(None=DotProduct(A=CameraVectorWS(), B=VertexNormalWS()))
  NamedRerouteDeclaration(None=Multiply(A=Multiply(A=@n68, B=@n69), B=NamedRerouteDeclaration(None=@n70) {name=Fresn Inver, nodeColor=<rgba 0.5234 1 0 1>})) {name=Tiled Depth, nodeColor=<rgba 0.2422 1 0 1>}
  Reroute(None=@n24)
 ]}
