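# Material function that gives moss and building materials a wet jungle-biome look through two outputs, both gated by the JungleTest switch.
# Moss darkens by AO, RVT height and surface up-facing, with optional sediment tint and RVT-driven leaks.
# Buildings get height ramps, roughness-masked green tint, detail normals, dirt overlay and optional world-aligned puddles on upward faces.
# Shading-path and quality switches skip effects on mobile or low tiers and limit details to higher ones.
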MaterialFunction /Game/Environments/Asteria/Rocks/Materials/MaterialFunctions/MF_JungleBiome
graph {nodes [
  n1 = NamedRerouteUsage()
  n2 = GetMaterialAttributes(None=NamedRerouteDeclaration(None=FunctionInput() {inputName=For Moss, inputType=FunctionInput_MaterialAttributes}) {name=Mat input, nodeColor=<rgba 0.0312 0 1 1>}) {attributeGetTypes [#1, #2, #3]}
  n3 = Multiply(A=@n2.Roughness) {constB=0.3}
  n4 = Multiply(A=@n2.Specular) {constB=0.5}
  n5 = GetMaterialAttributes(None=@n2.MaterialAttributes) {attributeGetTypes [#1, #3, #2]}
  n6 = Saturate(None=ComponentMask(None=VertexNormalWS()) {b=true})
  n7 = Multiply(
   A=@n5.BaseColor
   B=LinearInterpolate(
    A=ScalarParameter() {defaultValue=0.6615, group=Jungle, parameterName=Wet Rock}
    B=VectorParameter() {group=Jungle, parameterName=WetRockTop, defaultValue=<rgba 0.676 0.7344 0.547 1>}.RGB
    Alpha=@n6
   )
  )
  n8 = Multiply(A=@n5.Specular, B=ScalarParameter() {defaultValue=0.25, parameterName=JS})
  n9 = Multiply(
   A=@n5.Roughness
   B=LinearInterpolate(
    A=ScalarParameter() {defaultValue=0.7, group=Jungle, parameterName=JR}
    B=ScalarParameter() {defaultValue=0.8, group=Jungle, parameterName=JRTop}
    Alpha=@n6
   )
  )
  n10 = FunctionInput() {bUsePreviewValueAsDefault=true, inputName=AO, inputType=FunctionInput_Scalar, previewValue=<1 0 0 1>}
  n11 = FunctionInput() {bUsePreviewValueAsDefault=true, inputName=RVT_Height, inputType=FunctionInput_Scalar}
  n12 = Saturate(None=Divide(A=@n11, B=ScalarParameter() {defaultValue=155, group=Jungle, parameterName=RVT Jungle}))
  n13 = ShadingPathSwitch(
   Default=BlendMaterialAttributes(
    A=SetMaterialAttributes!(MaterialAttributes=@n2.MaterialAttributes, BaseColor=@n2.MaterialAttributes, Roughness=@n3, Specular=@n4) {attributeSetTypes [#1, #2, #3]}
    B=SetMaterialAttributes!(MaterialAttributes=@n5.MaterialAttributes, BaseColor=@n7, Specular=@n8, Roughness=@n9) {attributeSetTypes [#1, #3, #2]}
    Alpha=Multiply(A=@n10, B=@n12)
   )
   Mobile=@n1
  ) {desc=Migrated from FeatureLevelSwitch}
  n14 = TextureObject() {texture=/Game/Environments/Apollo/Sets/Abducted/Chunks/Textures/T_Abducted_Sediment_B}
  n15 = Constant() {r=3000}
  n16 = StaticSwitchParameter(
   True=MaterialFunctionCall!(
    Material=@n13
    Tint=MaterialFunctionCall(TextureObject=@n14, TextureSize=@n15) {materialFunction=/Game/Weapons/WeaponSkins/Wraps/MaterialFunctions/WorldAlignedTexture}.XY Texture
    TintMask=Constant() {r=0.5}
   ) {materialFunction=/Engine/Functions/MaterialLayerFunctions/MatLayerBlend_Tint}.Blended Material
   False=@n13
  ) {group=Jungle, parameterName=Sediment Layers}
  n17 = GetMaterialAttributes(None=@n16) {attributeGetTypes [#1, #3, #2]}
  n18 = Multiply(A=@n17.BaseColor, B=VectorParameter() {group=Jungle, parameterName=Leak Tint, defaultValue=<rgba 0.794 0.8872 0.3894 1>}.RGB)
  n19 = Multiply(A=@n17.Specular, B=ScalarParameter() {defaultValue=0.2, group=Jungle, parameterName=Leak S}) {constB=0.35}
  n20 = Multiply(A=@n17.Roughness, B=ScalarParameter() {defaultValue=0.2, group=Jungle, parameterName=Leak R})
  n21 = MaterialFunctionCall(Transition Phase=Divide(A=@n11, B=Constant() {r=2500}), Height Texture=@n17.Roughness, Contrast=Constant() {r=0.5}) {materialFunction=&HeightLerp}
  n22 = VirtualTextureFeatureSwitch!(
   No=@n16
   Yes=BlendMaterialAttributes(
    A=@n16
    B=SetMaterialAttributes(MaterialAttributes=@n17.MaterialAttributes, BaseColor=@n18, Specular=@n19, Roughness=@n20) {attributeSetTypes [#1, #3, #2]}
    Alpha=Saturate(None=@n21.Alpha)
   ) {vertexAttributeBlendType=UseA}
  )
  FunctionOutput(None=StaticSwitchParameter(True=ShadingPathSwitch(Default=@n22, Mobile=@n1), False=@n1) {group=Jungle, parameterName=JungleTest, sortPriority=-5})
  n23 = NamedRerouteUsage()
  n24 = FunctionInput() {inputName=For Buildings, inputType=FunctionInput_MaterialAttributes}
  n25 = NamedRerouteDeclaration(None=@n24) {name=Mat Input_1, nodeColor=<rgba 0 1 0.3672 1>}
  n26 = GetMaterialAttributes(None=@n25) {attributeGetTypes [#4, #1]}
  n27 = Multiply(A=TextureCoordinate(), B=ScalarParameter() {defaultValue=4, group=Jungle, parameterName=JungleDetailsScale, sortPriority=-2})
  n28 = TextureSample(UVs=@n27) {
   samplerSource=SSM_Wrap_WorldGroupSettings
   samplerType=SAMPLERTYPE_Normal
   texture=/Game/Environments/Asteria/Rocks/Textures/T_RockDetails2_N
  }
  n29 = Multiply(
   A=Multiply(A=@n28.RGB, B=Constant3Vector() {constant=<rgba 1 1 0 0>})
   B=ScalarParameter() {defaultValue=4, desc=Will use Layer 4 texture, group=Jungle, parameterName=Jungle Details Value, sortPriority=-3}
  )
  n30 = Transform(None=@n26.Normal) {transformSourceType=TRANSFORMSOURCE_World, transformType=TRANSFORM_Tangent}
  n31 = SetMaterialAttributes!(
   MaterialAttributes=@n26.MaterialAttributes
   Normal=Transform(None=Add(A=@n29, B=@n30))
   BaseColor=MaterialFunctionCall(
    Base=@n26.BaseColor
    Blend=TextureSample(UVs=@n27) {samplerType=SAMPLERTYPE_Masks, texture=/Game/Global/Base4Layers_2022/Textures/T_DirtLeaks_M}.R
   ) {materialFunction=&Blend_Overlay}.Result
  ) {attributeSetTypes [#4, #1]}
  n32 = GetMaterialAttributes(None=@n25) {attributeGetTypes [#1, #2, #3]}
  n33 = VectorParameter() {group=Jungle, parameterName=Jungle Ramp Bottom, defaultValue=<rgba 0.4135 0.4219 0.0902 0.5>}
  n34 = GetMaterialAttributes(None=@n25) {attributeGetTypes [#1, #2, #3]}
  n35 = VectorParameter() {group=Jungle, parameterName=Jungle Ramp Up, defaultValue=<rgba 1 1 1 1>}
  n36 = BlendMaterialAttributes(
   A=SetMaterialAttributes!(
    MaterialAttributes=@n32.MaterialAttributes
    BaseColor=Multiply(A=@n32.BaseColor, B=@n33.RGB)
    Roughness=Multiply(A=@n32.Roughness, B=@n33.A)
    Specular=Multiply(A=@n32.Specular, B=@n33.A)
   ) {attributeSetTypes [#1, #2, #3]}
   B=SetMaterialAttributes(
    MaterialAttributes=@n34.MaterialAttributes
    BaseColor=Multiply(A=@n34.BaseColor, B=@n35.RGB)
    Roughness=Multiply(A=@n34.Roughness, B=@n35.A)
    Specular=Multiply(A=@n34.Specular, B=@n35.A)
   ) {attributeSetTypes [#1, #2, #3]}
   Alpha=VirtualTextureFeatureSwitch(No=Constant(), Yes=@n12)
  )
  n37 = GetMaterialAttributes(None=@n36) {attributeGetTypes [#1, #2, #3]}
  n38 = VectorParameter() {group=Jungle, parameterName=Jungle Color, defaultValue=<rgba 0.3593 0.4219 0.0628 1>}
  n39 = VectorParameter() {group=Jungle, parameterName=Jungle Color Up, defaultValue=<rgba 0.4135 0.4219 0.0902 1>}
  n40 = Saturate(None=ComponentMask(None=VertexNormalWS()) {b=true})
  n41 = SetMaterialAttributes(
   MaterialAttributes=@n37.MaterialAttributes
   BaseColor=Multiply(A=LinearInterpolate(A=@n38.RGB, B=@n39.RGB, Alpha=@n40), B=@n37.BaseColor)
   Roughness=Multiply(A=@n37.Roughness, B=ScalarParameter() {defaultValue=0.35, group=Jungle, parameterName=Wet Rough Props})
   Specular=Multiply(A=@n37.Specular, B=ScalarParameter() {defaultValue=0.25, group=Jungle, parameterName=Wet Spec Prop})
  ) {attributeSetTypes [#1, #2, #3]}
  n42 = MaterialFunctionCall(
   Transition Phase=ScalarParameter() {defaultValue=0.68, group=Jungle, parameterName=Jungle Green Trans}
   Height Texture=LinearInterpolate(
    A=ScalarParameter() {defaultValue=-1, group=Jungle, parameterName=Mask Min}
    B=ScalarParameter() {defaultValue=0.85, group=Jungle, parameterName=Mask Max}
    Alpha=@n32.Roughness
   )
   Contrast=ScalarParameter() {defaultValue=0.5, group=Jungle, parameterName=Jungle Green Contrast}
  ) {materialFunction=&HeightLerp}
  n43 = GetMaterialAttributes(None=StaticSwitchParameter(True=@n25, False=BlendMaterialAttributes(A=@n41, B=@n36, Alpha=@n42.Alpha) {vertexAttributeBlendType=UseA}) {group=Jungle, parameterName=JungleDetails Only}) {attributeGetTypes [#4, #1]}
  n44 = Transform(None=@n43.Normal) {transformSourceType=TRANSFORMSOURCE_World, transformType=TRANSFORM_Tangent}
  n45 = Multiply(A=TextureCoordinate(), B=ScalarParameter() {defaultValue=4, group=Jungle, parameterName=JungleDetailsScale, sortPriority=-2})
  n46 = TextureSample(UVs=@n45) {
   samplerSource=SSM_Wrap_WorldGroupSettings
   samplerType=SAMPLERTYPE_Normal
   texture=&T_RockDetails2_N
  }
  n47 = Multiply(
   A=Multiply(A=@n46.RGB, B=Constant3Vector() {constant=<rgba 1 1 0 0>})
   B=ScalarParameter() {defaultValue=4, desc=Will use Layer 4 texture, group=Jungle, parameterName=Jungle Details Value, sortPriority=-3}
  )
  n48 = TextureObject() {samplerType=SAMPLERTYPE_Masks, texture=/Game/Global/Base4Layers_2022/Textures/T_DirtLeaks2_M}
  n49 = SetMaterialAttributes!(
   MaterialAttributes=@n43.MaterialAttributes
   Normal=Transform(None=Add(A=@n44, B=@n47))
   BaseColor=MaterialFunctionCall(
    Base=@n43.BaseColor
    Blend=TextureSample(UVs=@n45, Tex=@n48) {samplerType=SAMPLERTYPE_Masks, texture=&T_DirtLeaks2_M}.R
   ) {materialFunction=&Blend_Overlay}.Result
  ) {attributeSetTypes [#4, #1]}
  n50 = QualitySwitch(Default=@n49, Low=@n43.MaterialAttributes, High=@n43.MaterialAttributes, Medium=@n43.MaterialAttributes, Epic=@n49)
  n51 = GetMaterialAttributes(None=@n50) {attributeGetTypes [#1]}
  n52 = Constant3Vector() {constant=<rgba 0 0 1 0>}
  n53 = LinearInterpolate(
   A=@n52
   B=MaterialFunctionCall(
    TextureObject=TextureObject() {samplerType=SAMPLERTYPE_Normal, texture=/Game/Gadgets/Assets/Telephone_Booth/Textures/T_BlindsNormalNoise_N}
    TextureSize=ScalarParameter() {defaultValue=1024, group=Jungle, parameterName=Jungle Puddles Noise scale}
   ) {materialFunction=&WorldAlignedNormal}.Z Texture
   Alpha=ScalarParameter() {defaultValue=0.05, group=Jungle, parameterName=Puddles Normal Value}
  ) {constAlpha=0.7}
  n54 = SetMaterialAttributes!(
   MaterialAttributes=@n51.MaterialAttributes
   BaseColor=Multiply(A=VectorParameter() {group=Jungle, parameterName="Puddle ", defaultValue=<rgba 0.2517 0.1472 0.0723 1>}.RGB, B=@n51.BaseColor)
   Roughness=Constant() {r=0.05}
   Specular=Constant() {r=0.25}
   Normal=Transform(None=QualitySwitch(Default=@n53, Low=@n52, High=@n52, Medium=@n52, Epic=@n53))
  ) {attributeSetTypes [
    #1
    #2
    #3
    #4
   ]}
  n55 = VertexNormalWS()
  n56 = Constant3Vector() {constant=<rgba 0 0 -1 0>}
  n57 = Max(A=ScalarParameter() {defaultValue=0.5, group=Jungle, parameterName=Puddle Transition}, B=Saturate(None=DotProduct(A=@n55, B=@n56)))
  n58 = ScalarParameter() {defaultValue=5000, group=Jungle, parameterName=PuddleScale}
  n59 = WorldPosition()
  n60 = ActorPositionWS()
  n61 = Add(A=ComponentMask(None=Subtract(A=@n59.XYZ, B=@n60)) {b=true}, B=ScalarParameter() {group=Jungle, parameterName=Puddle Max Height Offset})
  n62 = ScalarParameter() {defaultValue=5, group=Jungle, parameterName=Puddle Z mask}
  n63 = Add(
   A=DotProduct(
    A=MaterialFunctionCall(TextureObject=@n48, TextureSize=@n58) {materialFunction=/Engine/Functions/Engine_MaterialFunctions01/Texturing/WorldAlignedTexture}.Z Texture
    B=VectorParameter() {group=Jungle, parameterName=Puddle Mask, defaultValue=<rgba 1 0 0 0>}.RGB
   )
   B=Saturate(None=Divide(A=@n61, B=@n62))
  )
  n64 = ScalarParameter() {group=Jungle, parameterName=Puddle Contrast}
  n65 = BlendMaterialAttributes(
   A=@n54
   B=@n50
   Alpha=Saturate(None=MaterialFunctionCall(Transition Phase=@n57, Height Texture=@n63, Contrast=@n64) {materialFunction=&HeightLerp}.Alpha)
  )
  n66 = Saturate(None=ComponentMask(None=@n55) {b=true})
  n67 = StaticSwitchParameter(
   True=QualitySwitch(Default=@n31, Low=@n25, High=@n25, Medium=@n25, Epic=@n31)
   False=StaticSwitchParameter(True=BlendMaterialAttributes(A=@n50, B=@n65, Alpha=@n66), False=@n50) {group=Jungle, parameterName=JunglePuddle}
  ) {group=Jungle, parameterName=JungleDetails Only}
  n68 = MaterialFunctionCall!(
   Low _ PC Low and Med=@n23
   "High _ XB1, PS4"=@n67
   "Epic _ PC High, XSX, PS5"=@n67
   --------
   "Mobile Low _ PC Perf, Android"=@n23
   Mobile Med _ Switch=@n23
  ) {materialFunction=/Game/Global/MaterialFunctions/TechArt/MF_Fortnite_QualitySwitch_MaterialAttributes}
  FunctionOutput(None=StaticSwitchParameter(True=@n68.Result, False=NamedRerouteUsage()) {group=Jungle, parameterName=JungleTest, sortPriority=-5}) {outputName=Buildings Jungl}
  n69 = GetMaterialAttributes(None=NamedRerouteUsage()) {attributeGetTypes [#1]}
  n70 = Constant3Vector() {constant=<rgba 0.173 0.2031 0.0302 1>}
  n71 = Constant() {r=0.55}
  n72 = Desaturation(None=@n69.BaseColor)
  n73 = Constant() {r=-1.25}
  NamedRerouteDeclaration(None=BlendMaterialAttributes(
    A=SetMaterialAttributes!(MaterialAttributes=@n69.MaterialAttributes, BaseColor=@n70) {attributeSetTypes [#1]}
    B=@n69.MaterialAttributes
    Alpha=MaterialFunctionCall(Transition Phase=@n71, Height Texture=@n72, Contrast=@n73) {materialFunction=&HeightLerp}.Alpha
   ) {vertexAttributeBlendType=UseA}) {name=Mobile Dirt Out, nodeColor=<rgba 1 0 0.6797 1>}
  Reroute(None=@n24)
  ShadingPathSwitch(Default=@n68.Result, Mobile=@n23) {desc=Migrated from FeatureLevelSwitch}
  StaticSwitchParameter(True=@n23, False=@n67) {group=JungleMobile, parameterName=TweakMobile}
 ]}
